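# Animation Blueprint for the MimicGlean Awe medium male player body in Fortnite, driving cosmetic secondary posing.
# It copies the parent mesh pose into a Control Rig, fed with front-end, DBNO, skydive, emote and rifle/pistol states.
# Each update derives weapon type and checks for the Takeout emote montage, which feeds its own rig input.
# On customization load, any tail part is hidden when the backbling is long and shown otherwise.
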AnimBlueprint /BRCosmetics/Characters/Player/Male/Medium/Bodies/M_MED_MimicGlean_Awe/Meshes/M_MED_MimicGlean_Awe_AnimBP
target_skeleton=&/M_MED_MimicGlean_Awe_Skeleton
graphs [
 AnimGraph {nodes [
   n1 = Control Rig!(
    Source=Copy Pose From Mesh() {bUseAttachedParent=true}
    Alpha
    InFrontEnd:bool=Get bIsInFrontEnd()
    DBNO:bool=Get bIsDBNO()
    IsRifle:bool=Get IsRifle() -> (IsRifle:bool)
    IsPistol:bool=Get isPistol() -> (isPistol:bool)
    ISSkydive:double=Get IsSkyDive() -> (IsSkyDive:double)
    IsEmote:bool=Get bIsPlayingEmote()
    Takeit:bool=Get TakeIT() -> (TakeIT:bool)
   ) -> (Pose) {controlRigAssetReference {controlRigAsset=&/M_MED_MimicGlean_Awe_CtrlRig}}
   Output Pose(Result=@n1) {layerGroup=None}
  ]}
 EventGraph {nodes [
   Event BlueprintUpdateAnimation()
   n1 = Get WeaponCoreAnim!() -> (WeaponCoreAnim)
   > Set isPistol(isPistol="Equal (Enum)"(A:EFortWeaponCoreAnimation Enum=@n1, B:EFortWeaponCoreAnimation Enum=Pistol))
   > Set IsRifle(IsRifle:bool="Equal (Enum)"(A=@n1, B=AssaultRifle)) -> (Output_Get:bool)
   n2 = GetOwnerSkeletalMeshComponent()
   > Is Valid(InputObject=@n2)
   > Set TakeIT(TakeIT:bool=Montage_IsPlaying(self=GetAnimInstance(self=@n2), Montage=/BRCosmetics/Animation/Game/MainPlayer/Emotes/Takeout/Emote_Takeout_CMM_M)) -> (Output_Get:bool)
   Event OnCharacterCustomizationLoaded()
   > n3 = Cast To PlayerPawn_Athena(Object=TryGetPawnOwner())
   > Branch(Condition=CheckForTail!(self:Custom Character Part Anim Instance Object Reference) -> (bHasTail:bool))
   > n4 = Branch(Condition=CheckForBackbling().bIsBackblingLong)
   n5 = GetSkeletalMeshForPartType(self=@n3, PartType=MiscOrTail)
   > SetVisibility!(execute, self:Scene Component Object Reference=@n5, bNewVisibility:bool=false, bPropagateToChildren:bool=false) -> (then)
   SetVisibility(execute=@n4.else, self=@n5, bNewVisibility=true, bPropagateToChildren=false)
   Event OnEndSkydiving()
   > Set IsSkyDive(IsSkyDive:double=0) -> (Output_Get:double)
   Event OnBeginSkydiving()
   > Set IsSkyDive(IsSkyDive=1)
  ]}
]
variables [
 isPistol {type=boolean}
 IsRifle {type=boolean}
 IsSkyDive {type=number}
 TakeIT {type=boolean}
]
