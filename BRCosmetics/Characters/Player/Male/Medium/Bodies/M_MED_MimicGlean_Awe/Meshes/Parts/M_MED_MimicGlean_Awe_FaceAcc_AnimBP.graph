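# Animation Blueprint for the face accessory part of the male medium MimicGlean Awe body, adding secondary motion to its hair bones.
# It copies the pose from the attached parent mesh, then runs sixteen chained AnimDynamics nodes, one per Hair bone, before output.
# All share identical tuning: root-relative simulation, angular spring with damping, gravity override, tight angular limits and scaled, clamped component motion.
# The EventGraph holds only an empty update event, with no variables.
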AnimBlueprint /BRCosmetics/Characters/Player/Male/Medium/Bodies/M_MED_MimicGlean_Awe/Meshes/Parts/M_MED_MimicGlean_Awe_FaceAcc_AnimBP
target_skeleton=&/M_MED_MimicGlean_Awe_FaceAcc_Skeleton
graphs [
 AnimGraph {nodes [
   n1 = Anim Dynamics - Bone: Hair_A(ComponentPose=Local To Component(LocalPose=Copy Pose From Mesh() {bUseAttachedParent=true}), Alpha=1, ExternalForce="0,0,0") {
    angularDampingOverride=0.9
    angularSpringConstant=110
    bAngularSpring=true
    bOverrideAngularDamping=true
    bUseGravityOverride=true
    bUsePlanarLimit=false
    simulationSpace=RootRelative
    boundBone {boneName=Hair_A}
    componentAppliedLinearAccClamp=<3000 3000 3500>
    componentLinearAccScale=<0.7 0.7 0.8>
    componentLinearVelScale=<0.7 0.7 0.8>
    gravityOverride {z=-980}
    physicsBodyDefinitions [{
      boundBone {boneName=Hair_A}
      boxExtents=<9.5 9.5 9.5>
      constraintSetup {coneAngle=1, angularLimitsMax=<1 16 16>, angularLimitsMin=<-1 -16 -16>, angularTarget=<1 0 0>}
      localJointOffset=<1.6 0 0>
     }]
    simSpaceSettings {maxAngularAcceleration=5000, maxAngularVelocity=5000, simSpaceAngularAlpha=0.22, externalAngularVelocity {z=0.1}}
   }
   n2 = Anim Dynamics - Bone: Hair_B(ComponentPose=@n1, Alpha=1, ExternalForce="0,0,0") {
    angularDampingOverride=0.9
    angularSpringConstant=110
    bAngularSpring=true
    bOverrideAngularDamping=true
    bUseGravityOverride=true
    bUsePlanarLimit=false
    simulationSpace=RootRelative
    boundBone {boneName=Hair_B}
    componentAppliedLinearAccClamp=<3000 3000 3500>
    componentLinearAccScale=<0.7 0.7 0.8>
    componentLinearVelScale=<0.7 0.7 0.8>
    gravityOverride {z=-980}
    physicsBodyDefinitions [{
      boundBone {boneName=Hair_B}
      boxExtents=<9.5 9.5 9.5>
      constraintSetup {coneAngle=1, angularLimitsMax=<1 16 16>, angularLimitsMin=<-1 -16 -16>, angularTarget=<1 0 0>}
      localJointOffset=<1.6 0 0>
     }]
    simSpaceSettings {maxAngularAcceleration=5000, maxAngularVelocity=5000, simSpaceAngularAlpha=0.22, externalAngularVelocity {z=0.1}}
   }
   n3 = Anim Dynamics - Bone: Hair_C(ComponentPose=@n2, Alpha=1, ExternalForce="0,0,0") {
    angularDampingOverride=0.9
    angularSpringConstant=110
    bAngularSpring=true
    bOverrideAngularDamping=true
    bUseGravityOverride=true
    bUsePlanarLimit=false
    simulationSpace=RootRelative
    boundBone {boneName=Hair_C}
    componentAppliedLinearAccClamp=<3000 3000 3500>
    componentLinearAccScale=<0.7 0.7 0.8>
    componentLinearVelScale=<0.7 0.7 0.8>
    gravityOverride {z=-980}
    physicsBodyDefinitions [{
      boundBone {boneName=Hair_C}
      boxExtents=<9.5 9.5 9.5>
      constraintSetup {coneAngle=1, angularLimitsMax=<1 16 16>, angularLimitsMin=<-1 -16 -16>, angularTarget=<1 0 0>}
      localJointOffset=<1.6 0 0>
     }]
    simSpaceSettings {maxAngularAcceleration=5000, maxAngularVelocity=5000, simSpaceAngularAlpha=0.22, externalAngularVelocity {z=0.1}}
   }
   n4 = Anim Dynamics - Bone: Hair_D(ComponentPose=@n3, Alpha=1, ExternalForce="0,0,0") {
    angularDampingOverride=0.9
    angularSpringConstant=110
    bAngularSpring=true
    bOverrideAngularDamping=true
    bUseGravityOverride=true
    bUsePlanarLimit=false
    simulationSpace=RootRelative
    boundBone {boneName=Hair_D}
    componentAppliedLinearAccClamp=<3000 3000 3500>
    componentLinearAccScale=<0.7 0.7 0.8>
    componentLinearVelScale=<0.7 0.7 0.8>
    gravityOverride {z=-980}
    physicsBodyDefinitions [{
      boundBone {boneName=Hair_D}
      boxExtents=<9.5 9.5 9.5>
      constraintSetup {coneAngle=1, angularLimitsMax=<1 16 16>, angularLimitsMin=<-1 -16 -16>, angularTarget=<1 0 0>}
      localJointOffset=<1.6 0 0>
     }]
    simSpaceSettings {maxAngularAcceleration=5000, maxAngularVelocity=5000, simSpaceAngularAlpha=0.22, externalAngularVelocity {z=0.1}}
   }
   n5 = Anim Dynamics - Bone: Hair_E(ComponentPose=@n4, Alpha=1, ExternalForce="0,0,0") {
    angularDampingOverride=0.9
    angularSpringConstant=110
    bAngularSpring=true
    bOverrideAngularDamping=true
    bUseGravityOverride=true
    bUsePlanarLimit=false
    simulationSpace=RootRelative
    boundBone {boneName=Hair_E}
    componentAppliedLinearAccClamp=<3000 3000 3500>
    componentLinearAccScale=<0.7 0.7 0.8>
    componentLinearVelScale=<0.7 0.7 0.8>
    gravityOverride {z=-980}
    physicsBodyDefinitions [{
      boundBone {boneName=Hair_E}
      boxExtents=<9.5 9.5 9.5>
      constraintSetup {coneAngle=1, angularLimitsMax=<1 16 16>, angularLimitsMin=<-1 -16 -16>, angularTarget=<1 0 0>}
      localJointOffset=<1.6 0 0>
     }]
    simSpaceSettings {maxAngularAcceleration=5000, maxAngularVelocity=5000, simSpaceAngularAlpha=0.22, externalAngularVelocity {z=0.1}}
   }
   n6 = Anim Dynamics - Bone: Hair_F(ComponentPose=@n5, Alpha=1, ExternalForce="0,0,0") {
    angularDampingOverride=0.9
    angularSpringConstant=110
    bAngularSpring=true
    bOverrideAngularDamping=true
    bUseGravityOverride=true
    bUsePlanarLimit=false
    simulationSpace=RootRelative
    boundBone {boneName=Hair_F}
    componentAppliedLinearAccClamp=<3000 3000 3500>
    componentLinearAccScale=<0.7 0.7 0.8>
    componentLinearVelScale=<0.7 0.7 0.8>
    gravityOverride {z=-980}
    physicsBodyDefinitions [{
      boundBone {boneName=Hair_F}
      boxExtents=<9.5 9.5 9.5>
      constraintSetup {coneAngle=1, angularLimitsMax=<1 16 16>, angularLimitsMin=<-1 -16 -16>, angularTarget=<1 0 0>}
      localJointOffset=<1.6 0 0>
     }]
    simSpaceSettings {maxAngularAcceleration=5000, maxAngularVelocity=5000, simSpaceAngularAlpha=0.22, externalAngularVelocity {z=0.1}}
   }
   n7 = Anim Dynamics - Bone: Hair_R(ComponentPose=@n6, Alpha=1, ExternalForce="0,0,0") {
    angularDampingOverride=0.9
    angularSpringConstant=110
    bAngularSpring=true
    bOverrideAngularDamping=true
    bUseGravityOverride=true
    bUsePlanarLimit=false
    simulationSpace=RootRelative
    boundBone {boneName=Hair_R}
    componentAppliedLinearAccClamp=<3000 3000 3500>
    componentLinearAccScale=<0.7 0.7 0.8>
    componentLinearVelScale=<0.7 0.7 0.8>
    gravityOverride {z=-980}
    physicsBodyDefinitions [{
      boundBone {boneName=Hair_R}
      boxExtents=<9.5 9.5 9.5>
      constraintSetup {coneAngle=1, angularLimitsMax=<1 16 16>, angularLimitsMin=<-1 -16 -16>, angularTarget=<1 0 0>}
      localJointOffset=<1.6 0 0>
     }]
    simSpaceSettings {maxAngularAcceleration=5000, maxAngularVelocity=5000, simSpaceAngularAlpha=0.22, externalAngularVelocity {z=0.1}}
   }
   n8 = Anim Dynamics - Bone: Hair_G(ComponentPose=@n7, Alpha=1, ExternalForce="0,0,0") {
    angularDampingOverride=0.9
    angularSpringConstant=110
    bAngularSpring=true
    bOverrideAngularDamping=true
    bUseGravityOverride=true
    bUsePlanarLimit=false
    simulationSpace=RootRelative
    boundBone {boneName=Hair_G}
    componentAppliedLinearAccClamp=<3000 3000 3500>
    componentLinearAccScale=<0.7 0.7 0.8>
    componentLinearVelScale=<0.7 0.7 0.8>
    gravityOverride {z=-980}
    physicsBodyDefinitions [{
      boundBone {boneName=Hair_G}
      boxExtents=<9.5 9.5 9.5>
      constraintSetup {coneAngle=1, angularLimitsMax=<1 16 16>, angularLimitsMin=<-1 -16 -16>, angularTarget=<1 0 0>}
      localJointOffset=<1.6 0 0>
     }]
    simSpaceSettings {maxAngularAcceleration=5000, maxAngularVelocity=5000, simSpaceAngularAlpha=0.22, externalAngularVelocity {z=0.1}}
   }
   n9 = Anim Dynamics - Bone: Hair_H(ComponentPose=@n8, Alpha=1, ExternalForce="0,0,0") {
    angularDampingOverride=0.9
    angularSpringConstant=110
    bAngularSpring=true
    bOverrideAngularDamping=true
    bUseGravityOverride=true
    bUsePlanarLimit=false
    simulationSpace=RootRelative
    boundBone {boneName=Hair_H}
    componentAppliedLinearAccClamp=<3000 3000 3500>
    componentLinearAccScale=<0.7 0.7 0.8>
    componentLinearVelScale=<0.7 0.7 0.8>
    gravityOverride {z=-980}
    physicsBodyDefinitions [{
      boundBone {boneName=Hair_H}
      boxExtents=<9.5 9.5 9.5>
      constraintSetup {coneAngle=1, angularLimitsMax=<1 16 16>, angularLimitsMin=<-1 -16 -16>, angularTarget=<1 0 0>}
      localJointOffset=<1.6 0 0>
     }]
    simSpaceSettings {maxAngularAcceleration=5000, maxAngularVelocity=5000, simSpaceAngularAlpha=0.22, externalAngularVelocity {z=0.1}}
   }
   n10 = Anim Dynamics - Bone: Hair_J(ComponentPose=@n9, Alpha=1, ExternalForce="0,0,0") {
    angularDampingOverride=0.9
    angularSpringConstant=110
    bAngularSpring=true
    bOverrideAngularDamping=true
    bUseGravityOverride=true
    bUsePlanarLimit=false
    simulationSpace=RootRelative
    boundBone {boneName=Hair_J}
    componentAppliedLinearAccClamp=<3000 3000 3500>
    componentLinearAccScale=<0.7 0.7 0.8>
    componentLinearVelScale=<0.7 0.7 0.8>
    gravityOverride {z=-980}
    physicsBodyDefinitions [{
      boundBone {boneName=Hair_J}
      boxExtents=<9.5 9.5 9.5>
      constraintSetup {coneAngle=1, angularLimitsMax=<1 16 16>, angularLimitsMin=<-1 -16 -16>, angularTarget=<1 0 0>}
      localJointOffset=<1.6 0 0>
     }]
    simSpaceSettings {maxAngularAcceleration=5000, maxAngularVelocity=5000, simSpaceAngularAlpha=0.22, externalAngularVelocity {z=0.1}}
   }
   n11 = Anim Dynamics - Bone: Hair_K(ComponentPose=@n10, Alpha=1, ExternalForce="0,0,0") {
    angularDampingOverride=0.9
    angularSpringConstant=110
    bAngularSpring=true
    bOverrideAngularDamping=true
    bUseGravityOverride=true
    bUsePlanarLimit=false
    simulationSpace=RootRelative
    boundBone {boneName=Hair_K}
    componentAppliedLinearAccClamp=<3000 3000 3500>
    componentLinearAccScale=<0.7 0.7 0.8>
    componentLinearVelScale=<0.7 0.7 0.8>
    gravityOverride {z=-980}
    physicsBodyDefinitions [{
      boundBone {boneName=Hair_K}
      boxExtents=<9.5 9.5 9.5>
      constraintSetup {coneAngle=1, angularLimitsMax=<1 16 16>, angularLimitsMin=<-1 -16 -16>, angularTarget=<1 0 0>}
      localJointOffset=<1.6 0 0>
     }]
    simSpaceSettings {maxAngularAcceleration=5000, maxAngularVelocity=5000, simSpaceAngularAlpha=0.22, externalAngularVelocity {z=0.1}}
   }
   n12 = Anim Dynamics - Bone: Hair_L(ComponentPose=@n11, Alpha=1, ExternalForce="0,0,0") {
    angularDampingOverride=0.9
    angularSpringConstant=110
    bAngularSpring=true
    bOverrideAngularDamping=true
    bUseGravityOverride=true
    bUsePlanarLimit=false
    simulationSpace=RootRelative
    boundBone {boneName=Hair_L}
    componentAppliedLinearAccClamp=<3000 3000 3500>
    componentLinearAccScale=<0.7 0.7 0.8>
    componentLinearVelScale=<0.7 0.7 0.8>
    gravityOverride {z=-980}
    physicsBodyDefinitions [{
      boundBone {boneName=Hair_L}
      boxExtents=<9.5 9.5 9.5>
      constraintSetup {coneAngle=1, angularLimitsMax=<1 16 16>, angularLimitsMin=<-1 -16 -16>, angularTarget=<1 0 0>}
      localJointOffset=<1.6 0 0>
     }]
    simSpaceSettings {maxAngularAcceleration=5000, maxAngularVelocity=5000, simSpaceAngularAlpha=0.22, externalAngularVelocity {z=0.1}}
   }
   n13 = Anim Dynamics - Bone: Hair_P(ComponentPose=@n12, Alpha=1, ExternalForce="0,0,0") {
    angularDampingOverride=0.9
    angularSpringConstant=110
    bAngularSpring=true
    bOverrideAngularDamping=true
    bUseGravityOverride=true
    bUsePlanarLimit=false
    simulationSpace=RootRelative
    boundBone {boneName=Hair_P}
    componentAppliedLinearAccClamp=<3000 3000 3500>
    componentLinearAccScale=<0.7 0.7 0.8>
    componentLinearVelScale=<0.7 0.7 0.8>
    gravityOverride {z=-980}
    physicsBodyDefinitions [{
      boundBone {boneName=Hair_P}
      boxExtents=<9.5 9.5 9.5>
      constraintSetup {coneAngle=1, angularLimitsMax=<1 16 16>, angularLimitsMin=<-1 -16 -16>, angularTarget=<1 0 0>}
      localJointOffset=<1.6 0 0>
     }]
    simSpaceSettings {maxAngularAcceleration=5000, maxAngularVelocity=5000, simSpaceAngularAlpha=0.22, externalAngularVelocity {z=0.1}}
   }
   n14 = Anim Dynamics - Bone: Hair_O(ComponentPose=@n13, Alpha=1, ExternalForce="0,0,0") {
    angularDampingOverride=0.9
    angularSpringConstant=110
    bAngularSpring=true
    bOverrideAngularDamping=true
    bUseGravityOverride=true
    bUsePlanarLimit=false
    simulationSpace=RootRelative
    boundBone {boneName=Hair_O}
    componentAppliedLinearAccClamp=<3000 3000 3500>
    componentLinearAccScale=<0.7 0.7 0.8>
    componentLinearVelScale=<0.7 0.7 0.8>
    gravityOverride {z=-980}
    physicsBodyDefinitions [{
      boundBone {boneName=Hair_O}
      boxExtents=<9.5 9.5 9.5>
      constraintSetup {coneAngle=1, angularLimitsMax=<1 16 16>, angularLimitsMin=<-1 -16 -16>, angularTarget=<1 0 0>}
      localJointOffset=<1.6 0 0>
     }]
    simSpaceSettings {maxAngularAcceleration=5000, maxAngularVelocity=5000, simSpaceAngularAlpha=0.22, externalAngularVelocity {z=0.1}}
   }
   n15 = Anim Dynamics - Bone: Hair_I(ComponentPose=@n14, Alpha=1, ExternalForce="0,0,0") {
    angularDampingOverride=0.9
    angularSpringConstant=110
    bAngularSpring=true
    bOverrideAngularDamping=true
    bUseGravityOverride=true
    bUsePlanarLimit=false
    simulationSpace=RootRelative
    boundBone {boneName=Hair_I}
    componentAppliedLinearAccClamp=<3000 3000 3500>
    componentLinearAccScale=<0.7 0.7 0.8>
    componentLinearVelScale=<0.7 0.7 0.8>
    gravityOverride {z=-980}
    physicsBodyDefinitions [{
      boundBone {boneName=Hair_I}
      boxExtents=<9.5 9.5 9.5>
      constraintSetup {coneAngle=1, angularLimitsMax=<1 16 16>, angularLimitsMin=<-1 -16 -16>, angularTarget=<1 0 0>}
      localJointOffset=<1.6 0 0>
     }]
    simSpaceSettings {maxAngularAcceleration=5000, maxAngularVelocity=5000, simSpaceAngularAlpha=0.22, externalAngularVelocity {z=0.1}}
   }
   n16 = Anim Dynamics - Bone: Hair_U(ComponentPose=@n15, Alpha=1, ExternalForce="0,0,0") {
    angularDampingOverride=0.9
    angularSpringConstant=110
    bAngularSpring=true
    bOverrideAngularDamping=true
    bUseGravityOverride=true
    bUsePlanarLimit=false
    simulationSpace=RootRelative
    boundBone {boneName=Hair_U}
    componentAppliedLinearAccClamp=<3000 3000 3500>
    componentLinearAccScale=<0.7 0.7 0.8>
    componentLinearVelScale=<0.7 0.7 0.8>
    gravityOverride {z=-980}
    physicsBodyDefinitions [{
      boundBone {boneName=Hair_U}
      boxExtents=<9.5 9.5 9.5>
      constraintSetup {coneAngle=1, angularLimitsMax=<1 16 16>, angularLimitsMin=<-1 -16 -16>, angularTarget=<1 0 0>}
      localJointOffset=<1.6 0 0>
     }]
    simSpaceSettings {maxAngularAcceleration=5000, maxAngularVelocity=5000, simSpaceAngularAlpha=0.22, externalAngularVelocity {z=0.1}}
   }
   Output Pose(Result=Component To Local(ComponentPose=@n16)) {layerGroup=None}
  ]}
 EventGraph {nodes [Event BlueprintUpdateAnimation()]}
]
variables []
